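# Masked unlit Niagara sprite material for the Boomer emote's UI effects, drawing either a texture, a procedural ring or a two-tone sphere.
# Use Ring and Use Sphere blend between modes, with subtracted sphere masks forming the ring and sphere outline.
# Dynamic parameters set ring size and randomize sphere colors.
# Optional time-of-day tinting precedes emissive adjustment, and lower quality or forward/mobile paths sample a lower-resolution mip.
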Material /BRCosmetics/Effects/Fort_Effects/Emotes/Boomer/M_Emote_Boomer_UI
graph {
 properties {bUsedWithNiagaraSprites=true, blendMode=BLEND_Masked, pixelDepthOffsetMode=PDOM_Legacy, shadingModel=MSM_Unlit, usedShadingModels=Unlit}
 nodes [
  n1 = Constant()
  n2 = Constant() {r=1}
  n3 = ShadingPathSwitch(Default=@n1, Deferred=@n1, Forward=@n2, Mobile=@n2)
  n4 = TextureSampleParameter2D!(UVs, Level=QualitySwitch(Default=@n3, Low=@n2, High=@n3, Medium=@n2, Epic=@n3), Apply View MipBias) {mipValueMode=TMVM_MipLevel, parameterName=UI, texture=&/T_Emote_Boomer_UI_Boom}
  n5 = TextureCoordinate()
  n6 = Constant() {r=0.5}
  n7 = Constant() {r=0.45}
  n8 = Constant() {r=1}
  n9 = SphereMask(A=@n5, B=@n6, Radius=@n7, Hardness=@n8)
  n10 = DynamicParameter() {defaultValue=<rgba 1 -0.2 1 1>, paramNames [ColorRandom,RingSize,ColorRandom2,Param4]}
  n11 = Subtract(A=@n9, B=SphereMask(A=@n5, B=@n6, Radius=Add(A=@n7, B=@n10.RingSize) {constB=-0.2}, Hardness=@n8))
  n12 = ScalarParameter() {group=Ring, parameterName=Use Ring}
  n13 = ScalarParameter() {group=Sphere, parameterName=Use Sphere}
  LinearInterpolate(A=LinearInterpolate(A=@n4.A, B=@n11, Alpha=@n12), B=@n9, Alpha=@n13)
  n14 = ParticleColor()
  n15 = VectorParameter() {group=Ring, parameterName=RingColor}
  n16 = VectorParameter() {group=Sphere, parameterName=SphereColor OutSide, defaultValue=<rgba 0.0194 0.0137 0.0103 1>}
  n17 = LinearInterpolate(
   A=VectorParameter() {group=Sphere, parameterName=SphereColor InSide1, defaultValue=<rgba 0.9647 0.0482 0.0222 1>}.RGB
   B=@n16.RGB
   Alpha=@n11
  )
  n18 = LinearInterpolate(
   A=VectorParameter() {group=Sphere, parameterName=SphereColor InSide2, defaultValue=<rgba 0.9473 0.7454 0.0052 1>}.RGB
   B=@n16.RGB
   Alpha=@n11
  )
  n19 = DynamicParameter() {defaultValue=<rgba 1 -0.2 1 1>, paramNames [ColorRandom,RingSize,ColorRandom2,Param4]}
  n20 = LinearInterpolate(
   A=LinearInterpolate(A=Multiply(A=@n4.RGB, B=@n14.RGB), B=Multiply(A=@n11, B=@n15.RGB), Alpha=@n12)
   B=LinearInterpolate(
    A=LinearInterpolate(A=@n17, B=@n18, Alpha=@n19.ColorRandom)
    B=LinearInterpolate(A=@n16.RGB, B=@n16.RGB, Alpha=@n11)
    Alpha=@n19.ColorRandom2
   )
   Alpha=@n13
  )
  n21 = MaterialFunctionCall(
   Color_Morning=VectorParameter() {group=ToD, parameterName=MorningTint, defaultValue=<rgba 1 1 1 1>}.RGB
   Color_Day=VectorParameter() {group=ToD, parameterName=DayTint, defaultValue=<rgba 1 1 1 1>}.RGB
   Color_Evening=VectorParameter() {group=ToD, parameterName=EveningTint, defaultValue=<rgba 0.75 0.75 0.75 1>}.RGB
   Color_Night=VectorParameter() {group=ToD, parameterName=NightTint, defaultValue=<rgba 0.5 0.5 0.5 1>}.RGB
  ) {materialFunction=&Color_Time_of_Day}
  MaterialFunctionCall(Emissive=LinearInterpolate(A=@n20, B=Multiply(A=@n21.Result, B=@n20), Alpha=ScalarParameter() {group=ToD, parameterName=Use ToD Multi})) {materialFunction=&MF_FX_EmissiveAdjustments}
  ScalarParameter() {defaultValue=-0.1, group=Ring, parameterName=RingSize}
 ]
}
